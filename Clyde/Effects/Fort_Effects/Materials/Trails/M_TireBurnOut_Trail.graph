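# Unlit two-sided modulate material for beam trails that darkens what lies behind it with tire burnout marks.
# An inverted, power-sharpened burnout texture on UV1 forms the mark, multiplied by a mask fading both ends of UV0's U axis.
# Depth fade, its distance varied by texture alpha, is skipped on low quality and mobile feature levels.
# Particle alpha blends from white, leaving the scene unchanged, to the darkened result.
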Material /Clyde/Effects/Fort_Effects/Materials/Trails/M_TireBurnOut_Trail
graph {
 properties {
  bUsedWithBeamTrails=true
  blendMode=BLEND_Modulate
  pixelDepthOffsetMode=PDOM_Legacy
  shadingModel=MSM_Unlit
  translucencyPass=MTP_BeforeDOF
  twoSided=true
  usedShadingModels=Unlit
 }
 nodes [
  n1 = TextureSample(UVs=TextureCoordinate() {coordinateIndex=1}) {texture=/Game/Effects/Fort_Effects/Textures/Masks/T_TireBurnOut}
  n2 = Clamp(None=Power(Base=OneMinus(None=@n1.RGB), Exp=ScalarParameter() {defaultValue=40, parameterName=sss}) {constExponent=40})
  n3 = DepthFade(Opacity=@n2, FadeDistance=LinearInterpolate(Alpha=@n1.A) {constA=3, constB=15})
  n4 = ComponentMask(None=TextureCoordinate()) {r=true}
  n5 = Multiply(A=Multiply(A=OneMinus(None=@n4), B=@n4), B=ScalarParameter() {defaultValue=100, parameterName=edgeWidthMultiply}) {constB=16}
  n6 = Multiply(A=QualitySwitch(Default=FeatureLevelSwitch(Default=@n3, ES2=@n2, ES3_1=@n2, SM4_REMOVED=@n3), Low=@n2), B=Clamp(None=Power(Base=@n5)))
  LinearInterpolate(B=Clamp(None=OneMinus(None=@n6)), Alpha=ParticleColor().A) {constA=1}
  ScalarParameter() {defaultValue=3, parameterName=FadeDistance}
 ]
}
